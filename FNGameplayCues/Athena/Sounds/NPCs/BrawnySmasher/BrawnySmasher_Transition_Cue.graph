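# Sound cue for the BrawnySmasher NPC's transition sound in Fortnite Athena gameplay cues.
# A single wave player plays one transition sound wave, with no randomization or modulation.
# It uses the BrawnySmasher attenuation asset and a melee-weapon sound class, primes on load, and overrides concurrency with a low cap on simultaneous instances.
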SoundCue /FNGameplayCues/Athena/Sounds/NPCs/BrawnySmasher/BrawnySmasher_Transition_Cue
root=WavePlayer_0
nodes [{id=WavePlayer_0, properties {soundWaveAssetPtr=&/BrawnySmasher_Transition_01}}]
properties {
 attenuationSettings=/Game/Athena/Sounds/NPCs/BrawnySmasher/BrawnySmasher_Att
 bOverrideConcurrency=true
 bPrimeOnLoad=true
 maxDistance=4000
 soundClassObject=&Fort_WeaponMelee_PSM
 volumeMultiplier=1
 concurrencyOverrides {maxCount=2}
}
